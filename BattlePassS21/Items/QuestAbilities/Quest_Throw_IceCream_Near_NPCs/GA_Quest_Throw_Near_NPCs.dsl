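; Quest gameplay ability that credits a player for throwing a PepperMint consumable (ice cream, per the quest name) near NPCs.
; On its trigger event it checks the instigator tags for that item, then caches the avatar pawn and its quest manager.
; It scans nearby pawns for one with the NPC tags but without the exclusion "Tags to Check", setting FoundNPC.
; A match sends a custom stat event to the quest manager.
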(event Ability|EventActivateAbilityFromEvent
  (params (EventData :type Gameplay%20Event%20Data%20Structure%20%28by%20ref%29))
  (bind (_eventtag _instigator _target _optionalobject _optionalobject2 _contexthandle _instigatortags) (Utilities|Struct|BreakGameplayEventData EventData))
  (|CallParentFunction EventData)
  (if (GameplayTags|HasTag _instigatortags "(TagName=\"Item.Consumable.PepperMint\")" true)
    (bind _asfort_player_pawn (Utilities|Casting|CastToFortPlayerPawn (Ability|GetAvatarActorfromActorInfo))
      (:then
        (bind _output_get (Variables|GAVars|Data|SetOwningPlayerPawn _asfort_player_pawn))
        (Utilities|IsValid (Fortnite|GetFortPlayerControllerfromActor _output_get)
          (:Is%20Valid
            (bind _output_get_2 (Variables|GAVars|Data|SetOwningPlayerQuestManager (Quests|GetQuestManager (Fortnite|GetFortPlayerControllerfromActor _output_get) "Invalid")))
            (Utilities|IsValid _output_get_2
              (:Is%20Valid
                (bind _outfortpawns (Fortnite|GetFortPawnsInRange (Transformation|GetActorLocation (Variables|GAVars|Data|GetOwningPlayerPawn)) 5e+02))
                (bind _array_element (Utilities|Array|ForEachLoopwithBreak _outfortpawns)
                  (:LoopBody
                    (bind _bdidactorhavetags (Class|AthenaQuestFunctionLibrary|AthenaCheckActorHasTags _array_element (Variables|Default|GetNPCTags) true))
                    (if _bdidactorhavetags
                      (bind _bdidactorhavetags_2 (Class|AthenaQuestFunctionLibrary|AthenaCheckActorHasTags _array_element (Variables|Default|GetTags%20to%20Check) true))
                      (if (not _bdidactorhavetags_2)
                        (Variables|Default|SetFoundNPC true))
                      (Ability|EndAbility))
                    (Ability|EndAbility))
                  (:Completed
                    _))
                (if (Variables|Default|GetFoundNPC)
                  (Quest|SendComplexCustomStatEvent (Variables|GAVars|Data|GetOwningPlayerQuestManager) 0 (Variables|Default|GetStatEventTags) "(GameplayTags=)" 1)
                  (Ability|EndAbility)
                  (else
                    (Ability|EndAbility))))
              (:Is%20Not%20Valid
                (Ability|EndAbility))))
          (:Is%20Not%20Valid
            (Ability|EndAbility))))
      (:CastFailed
        (Ability|EndAbility)))
    (else
      (Ability|EndAbility))))
